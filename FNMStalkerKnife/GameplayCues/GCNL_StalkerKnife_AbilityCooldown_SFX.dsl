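; Looping gameplay cue for the Stalker Knife ability cooldown that plays a 2D countdown-end sound when the cooldown expires.
; On loop start, only for a locally viewed player pawn, it sets a hard-coded duration and starts a one-second repeating timer.
; Each tick decrements the duration, playing the sound and clearing the timer at zero.
; Removing the cue also clears the timer.
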(event EventOnLoopingStart(EitherVFX-preferthisversion)
  (params (MyTarget :type Actor%20Object%20Reference) (Parameters :type Gameplay%20Cue%20Parameters%20Structure%20%28by%20ref%29) (ParticleComponents :type Array%20of%20FXSystem%20Component%20Object%20References) (AudioComponents :type Array%20of%20Audio%20Component%20Object%20References))
  (|CallParentFunction MyTarget Parameters ParticleComponents AudioComponents)
  (bind _asplayer_pawn_athena (Utilities|Casting|CastToPlayerPawn_Athena MyTarget))
  (if (Pawn|IsLocallyViewed _asplayer_pawn_athena)
    (Variables|Default|SetCooldownDuration 25.0)
    (bind _returnvalue (Utilities|Time|SetTimerbyEvent (delegate Custom|TickOnCooldown) 1.0 true))
    (Variables|Default|SetTimer _returnvalue)))

(event Custom|TickOnCooldown
  (Math|Float|DecrementFloat (Variables|Default|GetCooldownDuration)
    (:%20%20
      ;[0.5751 0.1986 0.8854 0.5] play sound on 0
      (if (<= (Variables|Default|GetCooldownDuration) 0)
        (Audio|PlaySound2D self "/FNMStalkerKnife/Assets/Audio/MSS_StalkerKnife_Secondary_Countdown_0s.MSS_StalkerKnife_Secondary_Countdown_0s")
        (Utilities|Time|ClearandInvalidateTimerbyHandle self (Variables|Default|GetTimer))))))

(event EventOnRemoval(EitherVFX-preferthisversion)
  (params (MyTarget :type Actor%20Object%20Reference) (Parameters :type Gameplay%20Cue%20Parameters%20Structure%20%28by%20ref%29) (ParticleComponents :type Array%20of%20FXSystem%20Component%20Object%20References) (AudioComponents :type Array%20of%20Audio%20Component%20Object%20References) (BurstCameraShakeInstance :type Camera%20Shake%20Base%20Object%20Reference) (BurstDecalInstance :type Decal%20Actor%20Object%20Reference))
  (|CallParentFunction MyTarget Parameters ParticleComponents AudioComponents BurstCameraShakeInstance BurstDecalInstance)
  (Utilities|Time|ClearandInvalidateTimerbyHandle self (Variables|Default|GetTimer)))

(fn ConstructionScript)
