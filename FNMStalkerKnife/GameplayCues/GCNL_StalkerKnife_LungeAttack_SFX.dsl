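; Looping gameplay cue notify that handles audio for the Stalker Knife lunge attack.
; On looping start it attaches an in-air attack start sound and a lunge attack sound to the target's root component.
; The in-air sound is stored and faded in, then faded out on removal if still valid.
; Removal also calls a second attached sound spawn, which has no sound asset assigned.
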(event EventOnLoopingStart(EitherVFX-preferthisversion)
  (params (MyTarget :type Actor%20Object%20Reference) (Parameters :type Gameplay%20Cue%20Parameters%20Structure%20%28by%20ref%29) (ParticleComponents :type Array%20of%20FXSystem%20Component%20Object%20References) (AudioComponents :type Array%20of%20Audio%20Component%20Object%20References))
  (bind _rootcomponent (Variables|/Script/Engine.Actor|GetRootComponent MyTarget))
  (|CallParentFunction MyTarget Parameters ParticleComponents AudioComponents)
  (bind _returnvalue (Audio|SpawnSoundAttached "/FNMStalkerKnife/Assets/Audio/MSS_StalkerKnife_InAir_Attack_start.MSS_StalkerKnife_InAir_Attack_start" _rootcomponent))
  (bind _output_get (Variables|Default|SetAirAttackSFX _returnvalue))
  (Audio|Components|Audio|FadeIn _output_get 0.1 1.0 0.0 "Linear" :!class /Script/Engine.AudioComponent)
  (Audio|SpawnSoundAttached "/FNMStalkerKnife/Assets/Audio/MSS_StalkerKnife_Lunge_Attack.MSS_StalkerKnife_Lunge_Attack" _rootcomponent))

(event EventOnRemoval(EitherVFX-preferthisversion)
  (params (MyTarget :type Actor%20Object%20Reference) (Parameters :type Gameplay%20Cue%20Parameters%20Structure%20%28by%20ref%29) (ParticleComponents :type Array%20of%20FXSystem%20Component%20Object%20References) (AudioComponents :type Array%20of%20Audio%20Component%20Object%20References) (BurstCameraShakeInstance :type Camera%20Shake%20Base%20Object%20Reference) (BurstDecalInstance :type Decal%20Actor%20Object%20Reference))
  (|CallParentFunction MyTarget Parameters ParticleComponents AudioComponents BurstCameraShakeInstance BurstDecalInstance)
  (Utilities|FlowControl|Sequence
    (:then_0
      (bind _airattacksfx (Variables|Default|GetAirAttackSFX)
        (:then
          (Audio|Components|Audio|FadeOut _airattacksfx 0.1 0.0 "Linear" :!class /Script/Engine.AudioComponent))
        (:else)))
    (:then_1
      (Audio|SpawnSoundAttached 0 (Variables|/Script/Engine.Actor|GetRootComponent MyTarget)))))

(fn ConstructionScript)
