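; Looping gameplay cue for the Stalker Knife transform, which shows a smoke transformation effect on the player who caused it.
; On application it takes the effect causer from the cue's effect context, casts it to a Fort player pawn and attaches the Niagara smoke system to that pawn's mesh at Root.
; The graph neither stores the spawned component nor has removal logic, and its other events are empty.
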(event EventBeginPlay)

(event Collision|EventActorBeginOverlap
  (params (OtherActor :type Actor%20Object%20Reference)))

(event EventTick
  (params (DeltaSeconds :type Float%20%28single-precision%29)))

(event EventOnApplication(EitherVFX-preferthisone)
  (params (MyTarget :type Actor%20Object%20Reference) (Parameters :type Gameplay%20Cue%20Parameters%20Structure%20%28by%20ref%29) (ParticleComponents :type Array%20of%20FXSystem%20Component%20Object%20References) (AudioComponents :type Array%20of%20Audio%20Component%20Object%20References) (BurstCameraShakeInstance :type Camera%20Shake%20Base%20Object%20Reference) (BurstDecalInstance :type Decal%20Actor%20Object%20Reference))
  (bind (_normalizedmagnitude _rawmagnitude _effectcontext) (Ability|GameplayCue|BreakGameplayCueParameters Parameters))
  (bind _asfort_player_pawn (Utilities|Casting|CastToFortPlayerPawn (Ability|EffectContext|GetEffectCauser _effectcontext)))
  (Niagara|SpawnSystemAttached "/FNMStalkerKnife/Assets/FX/NS_StalkerKnife_SmokeTransformation.NS_StalkerKnife_SmokeTransformation" (Variables|/Script/Engine.Character|GetMesh _asfort_player_pawn) "Root"))

(fn ConstructionScript)
