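; Gameplay cue for the StalkerKnife phone booth interaction, playing a sweep/ping visual with audio.
; On execute, it timestamps a material parameter on player 0's "Gameplay" mesh part and writes location, duration, radius and start time to the reused FringePlank Interrogate material parameter collection.
; It then attaches a 1P or 3P booth sound to the target, depending on whether the instigator is locally viewed, adding a 2D Ghostface voice line for 1P.
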(fn OnExecute
  (params (MyTarget :type Actor%20Object%20Reference) (Parameters :type Gameplay%20Cue%20Parameters%20Structure%20%28by%20ref%29))
  (bind (_normalizedmagnitude _rawmagnitude _effectcontext _matchedtagname _originaltag _aggregatedsourcetags _aggregatedtargettags _location) (Ability|GameplayCue|BreakGameplayCueParameters (Variables|GetParameters)))
  (|CallParentFunction MyTarget Parameters)
  (Utilities|FlowControl|Sequence
    (:then_0
      (bind _asfort_player_pawn (Utilities|Casting|CastToFortPlayerPawn (Game|GetPlayerPawn 0)))
      (Rendering|Material|SetScalarParameterValueonMaterials (Customization|GetSkeletalMeshforPartType _asfort_player_pawn "Gameplay" :!class /Script/FortniteGame.FortPlayerPawn) "ActivateTime" (Utilities|Time|GetGameTimeinSeconds 0))
      (Rendering|Material|SetVectorParameterValue 0 "/Game/Weapons/Prototype/FringePlank/PostProcess/MPC_FringePlank_Interrogate.MPC_FringePlank_Interrogate" "OriginPositionWS" (Math|Conversions|ToLinearColor(Vector) _location) :!class /Script/Engine.KismetMaterialLibrary)
      (Rendering|Material|SetScalarParameterValue 0 "/Game/Weapons/Prototype/FringePlank/PostProcess/MPC_FringePlank_Interrogate.MPC_FringePlank_Interrogate" "Duration" (Variables|Default|GetSweepDuration) :!class /Script/Engine.KismetMaterialLibrary)
      (Rendering|Material|SetScalarParameterValue 0 "/Game/Weapons/Prototype/FringePlank/PostProcess/MPC_FringePlank_Interrogate.MPC_FringePlank_Interrogate" "Radius" (ScalableFloat|GetValueatLevel (Variables|Default|GetPingRadius)) :!class /Script/Engine.KismetMaterialLibrary)
      (Rendering|Material|SetScalarParameterValue 0 "/Game/Weapons/Prototype/FringePlank/PostProcess/MPC_FringePlank_Interrogate.MPC_FringePlank_Interrogate" "ActivatedTimestamp" (Utilities|Time|GetGameTimeinSeconds 0) :!class /Script/Engine.KismetMaterialLibrary))
    (:then_1
      (bind _fortpawn (Class|FortQuestBlueprintLibrary|GetFortPawnFromObject (Ability|GameplayCue|GetInstigatorActor Parameters))
        (:Valid
          (Audio|SpawnSoundAttached (select (Pawn|IsLocallyViewed _fortpawn) "/FNMStalkerKnife/Assets/Audio/MSS_StalkerKnife_Booth_Interaction_1P.MSS_StalkerKnife_Booth_Interaction_1P" "/FNMStalkerKnife/Assets/Audio/MSS_StalkerKnife_Booth_Interaction_3P.MSS_StalkerKnife_Booth_Interaction_3P") (Variables|/Script/Engine.Actor|GetRootComponent (Variables|GetMyTarget)) "None" "0.000000,0.000000,100.000000")
          (if (Pawn|IsLocallyViewed _fortpawn)
            (Audio|PlaySound2D 0 "/FNMStalkerKnife/Assets/Audio/VO/VO_FNBR_VentQuip_Ghostface_PhoneBooth_1P.VO_FNBR_VentQuip_Ghostface_PhoneBooth_1P")))))
    (:then_2
      (return true))))

(fn ConstructionScript)
