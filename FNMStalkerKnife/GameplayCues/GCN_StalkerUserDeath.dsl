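; Gameplay cue notify for the death of the Stalker knife user, with its death sound still a placeholder.
; OnExecute calls the parent implementation, prints an orange debug string marking where the sound belongs, and returns true.
; No audio or effects are triggered in its own logic, and the construction script is empty.
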(fn OnExecute
  (params (MyTarget :type Actor%20Object%20Reference) (Parameters :type Gameplay%20Cue%20Parameters%20Structure%20%28by%20ref%29))
  (|CallParentFunction MyTarget Parameters)
  (Development|PrintString 0 "STALKER DEATH SOUND HERE" true true "(R=1.000000,G=0.526632,B=0.000000,A=1.000000)" 5.0)
  (return true))

(fn ConstructionScript)
