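; GCN_StalkerVictimRing is a gameplay cue notify for the Stalker Knife feature, named for a victim ring effect.
; On execute it calls the parent, resolves the cue's instigator to a Fort pawn and stores it as the Stalker.
; A check for whether that pawn is locally viewed leads to no actions, so this graph itself creates no ring.
; It always returns true, and the construction script is empty.
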(fn OnExecute
  (params (MyTarget :type Actor%20Object%20Reference) (Parameters :type Gameplay%20Cue%20Parameters%20Structure%20%28by%20ref%29))
  (|CallParentFunction MyTarget Parameters)
  (Utilities|FlowControl|Sequence
    (:then_0
      (bind _fortpawn (Class|FortQuestBlueprintLibrary|GetFortPawnFromObject (Ability|GameplayCue|GetInstigatorActor (Variables|GetParameters)))
        (:Valid
          (LocalVariables|SetAsStalker _fortpawn)
          (if (not (Pawn|IsLocallyViewed _fortpawn))
            _))))
    (:then_1
      (return true))))

(fn ConstructionScript)
